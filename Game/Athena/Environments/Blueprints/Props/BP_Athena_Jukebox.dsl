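; Jukebox music-player prop for Athena that plays party audio and reacts audibly to being hit.
; On begin play, the authority optionally applies an impact damage-reduction gameplay effect, while remotes start the MusicPartyProp component's party audio if no music track table is set.
; Damage triggers a change-track sound and a timeline, rate-scaled by PitchBendTime, that bends song pitch from lowered back toward normal.
; It also supplies interaction prompt text.
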(event EventBeginPlay
  (Networking|SwitchHasAuthority
    (:Authority
      (if (Variables|Default|GetbApplyDamageReductionEffect)
        (Environment|ApplyGlobalEnvironmentGameplayEffecttoActor self "/Game/Athena/Environments/Blueprints/Props/GE_JukeBoxImpactDamageReduction.GE_JukeBoxImpactDamageReduction_C")))
    (:Remote
      (Variables|Music|GetMusicTrackTable
        (:then)
        (:else
          (Class|BaseClass:MusicPartyProp_Base|StartParty (Variables|Default|GetMusicPartyPropComponent) (Variables|Default|GetMusicPartyAudio) 1.0))))))

(event EventOnDamagePlayEffects
  (params (Damage :type Float%20%28single-precision%29) (DamageTags :type Gameplay%20Tag%20Container%20Structure%20%28by%20ref%29) (Momentum :type Vector :default "0, 0, 0") (HitInfo :type Hit%20Result%20Structure%20%28by%20ref%29) (InstigatedBy :type Fort%20Pawn%20Object%20Reference) (DamageCauser :type Actor%20Object%20Reference) (EffectContext :type Gameplay%20Effect%20Context%20Handle%20Structure))
  ;[1 1 1] Pitch bend when taking damage
  (Utilities|FlowControl|Sequence
    (:then_0
      (Components|Timeline|SetPlayRate (Variables|BP_Athena_Jukebox|GetPitch%20Bend) (/ 1.0 (Variables|Music|GetPitchBendTime)))
      (bind (_direction _pitch) (Timeline|Pitch%20Bend PlayFromStart)
        (:Update
          (Audio|Components|Audio|SetPitchMultiplier (Variables|BuildingPropMusicPlayer|GetSongAudioComponent) (Math|Float|MapRangeClamped _pitch 0.0 1.0 0.8 1.0)))))
    (:then_1
      (Audio|PlaySoundatLocation (Variables|Default|GetChangeTrackSound) (Transformation|GetWorldLocation (Variables|BuildingPropMusicPlayer|GetSongAudioComponent))))))

(fn BlueprintGetInteractionString
  (params (InteractingPawn :type Fort%20Pawn%20Object%20Reference) (InteractionBeingAttempted :type EInteractionBeingAttempted%20Enum :default "FirstInteraction"))
  (return (Variables|GetInteractionText)))

(fn ConstructionScript)
